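# Material function for the TerryBroke cosmetic body that adjusts Base Color and Emissive, adding rim lighting and glowing red eyes.
# Left and right rim lights combine view-space positioned directions with smoothstepped fresnel gradients, optionally axis-aligned through a normal map.
# An FX texture darkens eye base color and drives red emissive.
# A time-of-day scalar, defaulting to night only, and a rim mask texture feed named reroutes that modulate the effects.
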MaterialFunction /BRCosmetics/Characters/Player/Male/Medium/Bodies/M_MED_TerryBroke/Materials/MF_TerryBroke
graph {nodes [
  n1 = GetMaterialAttributes(None=FunctionInput() {inputType=FunctionInput_MaterialAttributes}) {attributeGetTypes [#1, #2]}
  n2 = ScalarParameter() {controlType=Enumeration, enumeration=/Game/Global/Materials/Enums/E_Mat_OnOff, group=TerryBloke - RimLight, parameterName=IsTerryBlokeEyes}
  n3 = Multiply(A=NamedRerouteUsage(), B=@n2)
  n4 = TextureSample() {texture=/BRCosmetics/Characters/Player/Male/Medium/Bodies/M_MED_TerryBroke/Textures/T_TerryBroke_Head_FX}
  n5 = LinearInterpolate(A=@n1.BaseColor, Alpha=Multiply(A=@n4.G, B=@n3)) {constB=0}
  n6 = Multiply(A=@n5) {constB=0.1}
  n7 = NamedRerouteUsage()
  FunctionOutput(None=SetMaterialAttributes!(
    MaterialAttributes=@n1.MaterialAttributes
    BaseColor=LinearInterpolate(A=@n5, B=@n6, Alpha=@n7) {constAlpha=0}
    EmissiveColor=NamedRerouteUsage()
   ) {attributeSetTypes [#1, #2]})
  n8 = MaterialFunctionCall(
   Value_Morning=ScalarParameter() {group=TOD_Emissive, parameterName=Value_Morning}
   Value_Day=ScalarParameter() {group=TOD_Emissive, parameterName=Value_Day}
   Value_Evening=ScalarParameter() {group=TOD_Emissive, parameterName=Value_Evening}
   Value_Night=ScalarParameter() {defaultValue=1, group=TOD_Emissive, parameterName=Value_Night}
  ) {materialFunction=&Scalar_Time_of_Day}
  n9 = NamedRerouteUsage()
  n10 = LinearInterpolate(Alpha=OneMinus(None=StaticComponentMaskParameter(None=@n9) {defaultR=true, parameterName=RimLightMaskChannel})) {constA=0.45}
  NamedRerouteDeclaration(None=LinearInterpolate(A=@n8.Result, B=Multiply(A=@n8.Result, B=@n10), Alpha=ScalarParameter() {parameterName=hasDarkeningMask})) {name=TOD, nodeColor=<rgba 1 0 0.1172 1>}
  n11 = NamedRerouteUsage()
  n12 = ScalarParameter() {defaultValue=1, group=AlpacaLean - RimLight, parameterName=RimRight_Exponent, sortPriority=1}
  n13 = Subtract(
   A=TransformPosition(None=Constant3Vector() {constant=<rgba 100 50 50 0>}) {transformSourceType=TRANSFORMPOSSOURCE_View, transformType=TRANSFORMPOSSOURCE_World}
   B=WorldPosition().XYZ
  )
  n14 = PixelNormalWS()
  n15 = Saturate(None=DotProduct(A=Normalize(VectorInput=@n13), B=@n14))
  n16 = Multiply(
   A=SmoothStep(
    Min=ScalarParameter() {defaultValue=0.5, group=AlpacaLean - RimLight, parameterName=RimRIght_MinRange, sortPriority=2}
    Max=ScalarParameter() {defaultValue=0.8, group=AlpacaLean - RimLight, parameterName=RimRIght_MaxRange, sortPriority=3}
    Value=Power(Base=@n11, Exp=@n12) {constExponent=1}
   ) {constMax=0.8, constMin=0.5, constValue=0.4}
   B=@n15
  )
  n17 = VectorParameter() {group=AlpacaLean - RimLight, parameterName=RimRight_Color2, sortPriority=0, defaultValue=<rgba 0 0 0 1>}
  n18 = VectorParameter() {group=AlpacaLean - RimLight, parameterName=RimRight_Color, sortPriority=0, defaultValue=<rgba 0 0 0 1>}
  n19 = LinearInterpolate(A=@n17.RGB, B=@n18.RGB, Alpha=@n15)
  n20 = Subtract(
   A=TransformPosition(None=Constant3Vector() {constant=<rgba -100 50 50 0>}) {transformSourceType=TRANSFORMPOSSOURCE_View, transformType=TRANSFORMPOSSOURCE_World}
   B=WorldPosition().XYZ
  )
  n21 = NamedRerouteUsage()
  n22 = ScalarParameter() {defaultValue=1, group=AlpacaLean - RimLight, parameterName=RimLeft_Exponent, sortPriority=11}
  n23 = Multiply(
   A=DotProduct(A=@n14, B=Normalize(VectorInput=@n20))
   B=SmoothStep(
    Min=ScalarParameter() {defaultValue=0.5, group=AlpacaLean - RimLight, parameterName=RimLeft_MinRange, sortPriority=12}
    Max=ScalarParameter() {defaultValue=0.8, group=AlpacaLean - RimLight, parameterName=RimLeft_MaxRange, sortPriority=13}
    Value=Power(Base=@n21, Exp=@n22) {constExponent=1}
   ) {constMax=0.8, constMin=0.5, constValue=0.4}
  )
  n24 = VectorParameter() {group=AlpacaLean - RimLight, parameterName=RimLeft_Color, sortPriority=10, defaultValue=<rgba 0.9301 0.0998 0 1>}
  NamedRerouteDeclaration(None=Max(A=Multiply(A=@n16, B=@n19), B=Multiply(A=@n23, B=@n24.RGB))) {name=RimLight, nodeColor=<rgba 1 0 0.1172 1>}
  n25 = TextureSampleParameter2D() {
   parameterName=FresnelAxisMap
   samplerType=SAMPLERTYPE_Normal
   texture=/BRCosmetics/Characters/Player/Male/Medium/Bodies/M_MED_TerryBroke/Textures/T_TerryBroke_AxisFresnel
  }
  n26 = OneMinus(None=MaterialFunctionCall(Axis=Transform(None=@n25.RGB)) {materialFunction=&AxisAlignedFresnel}.Fresnel)
  n27 = OneMinus(None=DotProduct(A=CameraVectorWS(), B=PixelNormalWS()))
  n28 = PreSkinnedNormal()
  n29 = Abs(None=StaticComponentMaskParameter(None=VertexInterpolator(VS=@n28).PS) {defaultB=true, parameterName=FresnelAxis})
  NamedRerouteDeclaration(None=StaticSwitchParameter(True=LinearInterpolate(A=@n26, B=@n27, Alpha=@n29), False=@n27) {parameterName=UseAxisAlignedFresnelRimLight}) {name=RimLight_FresenelGradient, nodeColor=<rgba 0.0312 0 1 1>}
  n30 = Max(A=@n16, B=@n23)
  n31 = OneMinus(None=NamedRerouteUsage())
  NamedRerouteDeclaration(None=LinearInterpolate(A=@n30, B=Multiply(A=@n30, B=@n31), Alpha=ScalarParameter() {parameterName=hasMaskRimlight})) {name=RimLight_Alpha, nodeColor=<rgba 0 0.7891 1 1>}
  n32 = LinearInterpolate(A=@n1.EmissiveColor, B=Multiply(A=@n1.EmissiveColor) {constB=0.1}, Alpha=NamedRerouteUsage())
  n33 = Constant3Vector() {constant=<rgba 1 0 0 1>}
  n34 = Constant3Vector() {constant=<rgba 1 0.1089 0.0781 1>}
  n35 = Saturate(None=MaterialFunctionCall(In=@n4.R, Contrast=Constant() {r=0.0947}) {materialFunction=&CheapContrast}.Result)
  n36 = Multiply(A=@n4.B) {constB=0.5}
  n37 = Multiply(A=LinearInterpolate(A=@n32, B=LinearInterpolate(A=@n33, B=@n34, Alpha=@n35), Alpha=Add(A=@n4.G, B=@n36)), B=@n3)
  n38 = NamedRerouteUsage()
  n39 = ScalarParameter() {
   controlType=Enumeration
   defaultValue=1
   enumeration=&E_Mat_OnOff
   group=TerryBloke - RimLight
   parameterName=hasRimLight
  }
  n40 = Multiply(A=Multiply(A=NamedRerouteUsage(), B=NamedRerouteUsage()), B=@n39)
  NamedRerouteDeclaration(None=Add(A=Add(A=@n32, B=@n37), B=Multiply(A=@n38, B=@n40))) {name=FinalEmissive, nodeColor=<rgba 1 0.9844 0 1>}
  n41 = TextureSampleParameter2D() {
   parameterName=RimLightMask
   samplerType=SAMPLERTYPE_LinearGrayscale
   texture=/BRCosmetics/Characters/Player/Male/Medium/Bodies/M_MED_TerryBroke/Textures/T_TerryBroke_FaceAcc_FX
  }
  NamedRerouteDeclaration(None=@n41.RGB) {name=RimlightMask, nodeColor=<rgba 1 0.2578 0 1>}
  TextureSample() {
   samplerType=SAMPLERTYPE_LinearGrayscale
   texture=&T_TerryBroke_FaceAcc_FX
  }
 ]}
