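# Post-process style Anim Blueprint for the TerryBroke body mesh that adds secondary motion to wool and tail bones.
# It copies the attached parent's pose, chains root-relative spring AnimDynamics nodes, then runs the character's Control Rig.
# Emote events set baletSpin for one ballerina keytar montage, feeding the rig's emote input.
# Cape and head-limit variables are set but unread here; a head-bone offset and duplicate tail nodes sit outside the output chain.
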AnimBlueprint /BRCosmetics/Characters/Player/Male/Medium/Bodies/M_MED_TerryBroke/Meshes/M_MED_TerryBroke_AnimBP
target_skeleton=&/M_MED_TerryBroke_Skeleton
graphs [
 AnimGraph {nodes [
   n1 = Anim Dynamics - Bone: tail_1(Alpha=1, ExternalForce="0,0,0") {
    angularDampingOverride=0.9999
    bAngularSpring=true
    bUseGravityOverride=true
    boundBone {boneName=tail_1}
    gravityOverride {z=-980}
    physicsBodyDefinitions [{
      boundBone {boneName=tail_1}
      boxExtents=<10 10 10>
      constraintSetup {angularTargetAxis=AxisY, angularLimitsMax=<0 0 22>, angularLimitsMin=<0 0 -20>, angularTarget=<0 1 0>}
      localJointOffset=<0 -1 0>
     }]
   }
   Anim Dynamics - Bone: tail_2(ComponentPose=@n1, Alpha=1, ExternalForce="0,0,0") {
    angularDampingOverride=0.9999
    bAngularSpring=true
    bUseGravityOverride=true
    boundBone {boneName=tail_2}
    gravityOverride {z=-980}
    physicsBodyDefinitions [{
      boundBone {boneName=tail_2}
      boxExtents=<10 10 10>
      constraintSetup {angularTargetAxis=AxisY, angularLimitsMax=<0 0 22>, angularLimitsMin=<0 0 -20>, angularTarget=<0 1 0>}
      localJointOffset=<0 -1 0>
     }]
   }
   "Transform (Modify) Bone - Bone: head"(Alpha=Get headlock() -> (headlock:double), Translation="0,0,0", Rotation="-23.800,-103.114,25.311", Scale="1,1,1") {rotationMode=BMM_Additive, rotationSpace=BCS_ParentBoneSpace, boneToModify {boneName=head}, rotation=<rot -23.8 -103.114 25.311>}
   n2 = Anim Dynamics - Bone: dyn_wool_04(ComponentPose=Local To Component(LocalPose=Copy Pose From Mesh() {bUseAttachedParent=true}), Alpha=1, ExternalForce="0,0,0") {
    angularDampingOverride=0.99
    angularSpringConstant=1500
    bAngularSpring=true
    bOverrideAngularDamping=true
    bUseGravityOverride=true
    simulationSpace=RootRelative
    boundBone {boneName=dyn_wool_04}
    componentAppliedLinearAccClamp=<3000 3000 10000>
    componentLinearAccScale=<0.3 0.3 0.5>
    componentLinearVelScale=<0.3 0.3 1>
    gravityOverride {z=-980}
    physicsBodyDefinitions [{
      boundBone {boneName=dyn_wool_04}
      boxExtents=<13 13 13>
      constraintSetup {angularLimitsMax=<10 10 10>, angularLimitsMin=<-10 -10 -10>, angularTarget=<1 0 0>}
      localJointOffset=<0.2 0 0>
     }]
    relativeSpaceBone {boneName=spine_05}
    simSpaceSettings {maxAngularAcceleration=40, maxAngularVelocity=8, simSpaceAngularAlpha=0.3}
   }
   n3 = Anim Dynamics - Bone: dyn_wool_05(ComponentPose=@n2, Alpha=1, ExternalForce="0,0,0") {
    angularDampingOverride=0.999
    angularSpringConstant=1500
    bAngularSpring=true
    bOverrideAngularDamping=true
    bUseGravityOverride=true
    simulationSpace=RootRelative
    boundBone {boneName=dyn_wool_05}
    componentAppliedLinearAccClamp=<3000 3000 10000>
    componentLinearAccScale=<0.3 0.3 0.5>
    componentLinearVelScale=<0.3 0.3 1>
    gravityOverride {z=-980}
    physicsBodyDefinitions [{
      boundBone {boneName=dyn_wool_05}
      boxExtents=<11 11 11>
      constraintSetup {angularLimitsMax=<10 10 10>, angularLimitsMin=<-10 -10 -10>, angularTarget=<1 0 0>}
      localJointOffset=<0.2 0 0>
     }]
    relativeSpaceBone {boneName=spine_05}
    simSpaceSettings {maxAngularAcceleration=40, maxAngularVelocity=8, simSpaceAngularAlpha=0.3}
   }
   n4 = Anim Dynamics - Bone: dyn_wool_02(ComponentPose=@n3, Alpha=1, ExternalForce="0,0,0") {
    angularDampingOverride=0.99
    angularSpringConstant=1500
    bAngularSpring=true
    bOverrideAngularDamping=true
    bUseGravityOverride=true
    simulationSpace=RootRelative
    boundBone {boneName=dyn_wool_02}
    componentAppliedLinearAccClamp=<3000 3000 10000>
    componentLinearAccScale=<0.3 0.3 0.5>
    componentLinearVelScale=<0.3 0.3 1>
    gravityOverride {z=-980}
    physicsBodyDefinitions [{
      boundBone {boneName=dyn_wool_02}
      boxExtents=<13 13 13>
      constraintSetup {angularLimitsMax=<10 10 10>, angularLimitsMin=<-10 -10 -10>, angularTarget=<1 0 0>}
      localJointOffset=<0.2 0 0>
     }]
    relativeSpaceBone {boneName=spine_05}
    simSpaceSettings {maxAngularAcceleration=40, maxAngularVelocity=8, simSpaceAngularAlpha=0.3}
   }
   n5 = Anim Dynamics - Bone: dyn_wool_03(ComponentPose=@n4, Alpha=1, ExternalForce="0,0,0") {
    angularDampingOverride=0.999
    angularSpringConstant=1500
    bAngularSpring=true
    bOverrideAngularDamping=true
    bUseGravityOverride=true
    simulationSpace=RootRelative
    boundBone {boneName=dyn_wool_03}
    componentAppliedLinearAccClamp=<3000 3000 10000>
    componentLinearAccScale=<0.3 0.3 0.5>
    componentLinearVelScale=<0.3 0.3 1>
    gravityOverride {z=-980}
    physicsBodyDefinitions [{
      boundBone {boneName=dyn_wool_03}
      boxExtents=<11 11 11>
      constraintSetup {angularLimitsMax=<10 10 10>, angularLimitsMin=<-10 -10 -10>, angularTarget=<1 0 0>}
      localJointOffset=<0.2 0 0>
     }]
    relativeSpaceBone {boneName=spine_05}
    simSpaceSettings {maxAngularAcceleration=40, maxAngularVelocity=8, simSpaceAngularAlpha=0.3}
   }
   n6 = Anim Dynamics - Bone: dyn_wool_01(ComponentPose=@n5, Alpha=1, ExternalForce="0,0,0") {
    angularDampingOverride=0.99
    angularSpringConstant=1500
    bAngularSpring=true
    bOverrideAngularDamping=true
    bUseGravityOverride=true
    simulationSpace=RootRelative
    boundBone {boneName=dyn_wool_01}
    componentAppliedLinearAccClamp=<3000 3000 10000>
    componentLinearAccScale=<0.3 0.3 0.5>
    componentLinearVelScale=<0.3 0.3 1>
    gravityOverride {z=-980}
    physicsBodyDefinitions [{
      boundBone {boneName=dyn_wool_01}
      boxExtents=<13 13 13>
      constraintSetup {angularLimitsMax=<10 5 10>, angularLimitsMin=<-10 -15 -10>, angularTarget=<1 0 0>}
      localJointOffset=<0.2 0 0>
     }]
    relativeSpaceBone {boneName=spine_05}
    simSpaceSettings {maxAngularAcceleration=40, maxAngularVelocity=8, simSpaceAngularAlpha=0.3}
   }
   n7 = Anim Dynamics - Bone: tail_1(ComponentPose=@n6, Alpha=1, ExternalForce="0,0,0") {
    angularDampingOverride=0.99
    angularSpringConstant=2000
    bAngularSpring=true
    bOverrideAngularDamping=true
    bUseGravityOverride=true
    simulationSpace=RootRelative
    boundBone {boneName=tail_1}
    componentAppliedLinearAccClamp=<3000 3000 10000>
    componentLinearAccScale=<0.3 0.3 0.5>
    componentLinearVelScale=<0.3 0.3 1>
    gravityOverride {z=-980}
    physicsBodyDefinitions [{
      boundBone {boneName=tail_1}
      boxExtents=<18 18 18>
      constraintSetup {angularTargetAxis=AxisY, angularLimitsMax=<2 2 22>, angularLimitsMin=<-2 -2 -15>, angularTarget=<0 1 0>}
      localJointOffset=<0 -0.6 0>
     }]
    relativeSpaceBone {boneName=spine_05}
    simSpaceSettings {maxAngularAcceleration=40, maxAngularVelocity=8, simSpaceAngularAlpha=0.3}
   }
   n8 = Anim Dynamics - Bone: tail_2(ComponentPose=@n7, Alpha=1, ExternalForce="0,0,0") {
    angularDampingOverride=0.9999
    angularSpringConstant=3000
    bAngularSpring=true
    bOverrideAngularDamping=true
    bUseGravityOverride=true
    simulationSpace=RootRelative
    boundBone {boneName=tail_2}
    componentAppliedLinearAccClamp=<3000 3000 10000>
    componentLinearAccScale=<0.3 0.3 0.5>
    componentLinearVelScale=<0.3 0.3 1>
    gravityOverride {z=-980}
    physicsBodyDefinitions [{
      boundBone {boneName=tail_2}
      boxExtents=<14 14 14>
      constraintSetup {angularTargetAxis=AxisY, angularLimitsMax=<10 10 22>, angularLimitsMin=<-10 -10 -22>, angularTarget=<0 1 0>}
      localJointOffset=<0 -0.6 0>
     }]
    relativeSpaceBone {boneName=spine_05}
    simSpaceSettings {maxAngularAcceleration=40, maxAngularVelocity=8, simSpaceAngularAlpha=0.3}
   }
   n9 = Control Rig!(Source=Component To Local(ComponentPose=@n8), Alpha, emote:bool=Get baletSpin() -> (baletSpin:bool)) -> (Pose) {controlRigAssetReference {controlRigAsset=&/M_MED_TerryBroke_CtrlRig}}
   Output Pose(Result=@n9) {layerGroup=None}
  ]}
 EventGraph {nodes [
   Event OnEmoteStarted!() -> (OutputDelegate, then, MontageItemDef:Fort Item Definition Object Reference) {is=title K2Node_Event}
   n1 = GetCurrentActiveMontage!(self:Anim Instance Object Reference=GetOwnerAnimInstance!(self:Custom Character Part Anim Instance Object Reference) -> (ReturnValue:Anim Instance Object Reference)) -> (ReturnValue:Anim Montage Object Reference)
   > Set baletSpin(baletSpin:bool="Equal Exactly (String)"(A=GetDisplayName(Object=@n1), B=Emote_Keytar_Spawned_Ballerina_CMM_M)) -> (Output_Get:bool)
   Event OnEmoteStopped!() -> (OutputDelegate, then, MontageItemDef:Fort Item Definition Object Reference) {is=title K2Node_Event}
   > Set baletSpin(baletSpin=false)
   Event OnCharacterCustomizationLoaded()
   > Sequence!(execute) -> (then_0, then_1, then_2:Exec)
   > Set IsWearingCape(IsWearingCape:bool=CheckForBackbling().bIsBackblingACape) -> (Output_Get:bool)
   Reroute Node()
   n2 = GetOwnerSkeletalMeshComponent()
   n3 = BreakRotator(InRot=BreakTransform(InTransform=GetDeltaTransformFromRefPose(self=@n2, BoneName=head, BaseName=neck_02)).Rotation)
   n4 = Get Z percentageEmote() -> (Z percentageEmote:float)
   n5 = Get X percentageEmote() -> (X percentageEmote:float)
   n6 = Get Y percentageEmote() -> (Y percentageEmote:float)
   > Set headRotLimit(headRotLimit:Rotator=MakeRotator(
     Roll=float * float!(A=@n3, B=-1, C:double=@n5) -> (ReturnValue)
     Pitch=float * float(A=@n3.Pitch, B=-1, C=@n6)
     Yaw=float * float(A=@n3.Yaw, B=-1, C=@n4)
    )) -> (Output_Get:Rotator)
   n7 = Get Z percentageEmote_no_colar() -> (Z percentageEmote_no_colar:float)
   > Set headRotLimit_no_colar(headRotLimit_no_colar:Rotator=MakeRotator(Roll=@n3, Pitch=@n3.Pitch, Yaw=float * float(A=@n3.Yaw, B=-1, C=@n7))) -> (Output_Get:Rotator)
   Event BlueprintUpdateAnimation()
  ]}
]
variables [
 baletSpin {type=boolean}
 Custom Boots Equip {type=boolean}
 CustomBootsEquip {type=boolean}
 headlock {type=number}
 headRotLimit {type=Rotator}
 headRotLimit_no_colar {type=Rotator}
 IsWearingCape {type=boolean}
 X percentageEmote {type=number, default=1}
 Y percentageEmote {type=number, default=1}
 Z percentageEmote {type=number, default=1}
 Z percentageEmote_no_colar {type=number, default=-1}
]
