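# Post-customization Anim Blueprint for TerryBroke's face accessory mesh, adding secondary motion to its collar bones.
# It copies the attached parent's pose, layers both clavicle branches from the customization-supplied Body mesh, then runs a Control Rig.
# Chained spring-driven AnimDynamics nodes simulate left, right and back collar bones relative to spine_05.
# Updates compute inverted, scaled head rotation and remapped forearm-to-head distances, though the AnimGraph shows no reads of them.
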AnimBlueprint /BRCosmetics/Characters/Player/Male/Medium/Bodies/M_MED_TerryBroke/Meshes/Parts/M_MED_TerryBroke_FaceAcc_AnimBP
target_skeleton=&/M_MED_TerryBroke_FaceAcc_Skeleton
graphs [
 AnimGraph {nodes [
   n1 = Get Body() -> (Body:Skeletal Mesh Component Object Reference)
   n2 = Control Rig(Source=Layered blend per bone(
     BasePose=Copy Pose From Mesh() {bUseAttachedParent=true}
     BlendPoses_0=Copy Pose From Mesh(SourceMeshComponent=@n1) {bUseAttachedParent=true}
    ) {layerSetup [{branchFilters [{boneName=clavicle_l}, {boneName=clavicle_r}]}]}) {controlRigAssetReference {controlRigAsset=&/M_MED_TerryBroke_FaceAcc_CtrlRig1}}
   n3 = Anim Dynamics - Bone: dyn_collar_l_02(ComponentPose=Local To Component(LocalPose=@n2), Alpha=1, ExternalForce="0,0,0") {
    angularDampingOverride=0.99
    angularSpringConstant=200
    bAngularSpring=true
    bOverrideAngularDamping=true
    bUseGravityOverride=true
    simulationSpace=RootRelative
    boundBone {boneName=dyn_collar_l_02}
    componentAppliedLinearAccClamp=<3000 3000 10000>
    componentLinearAccScale=<0.3 0.3 0.5>
    componentLinearVelScale=<0.3 0.3 1>
    gravityOverride {z=-980}
    physicsBodyDefinitions [{
      boundBone {boneName=dyn_collar_l_02}
      boxExtents=<12 12 12>
      constraintSetup {angularLimitsMax=<0 15 0>, angularLimitsMin=<0 -15 0>, angularTarget=<1 0 0>}
      localJointOffset=<1 0 0>
     }]
    relativeSpaceBone {boneName=spine_05}
    simSpaceSettings {maxAngularAcceleration=40, maxAngularVelocity=8, simSpaceAngularAlpha=0.3}
   }
   n4 = Anim Dynamics - Bone: dyn_collar_l_03(ComponentPose=@n3, Alpha=1, ExternalForce="0,0,0") {
    angularDampingOverride=0.99
    angularSpringConstant=200
    bAngularSpring=true
    bOverrideAngularDamping=true
    bUseGravityOverride=true
    simulationSpace=RootRelative
    boundBone {boneName=dyn_collar_l_03}
    componentAppliedLinearAccClamp=<3000 3000 10000>
    componentLinearAccScale=<0.3 0.3 0.5>
    componentLinearVelScale=<0.3 0.3 1>
    gravityOverride {z=-980}
    physicsBodyDefinitions [{
      boundBone {boneName=dyn_collar_l_03}
      boxExtents=<12 12 12>
      constraintSetup {angularLimitsMax=<5 15 5>, angularLimitsMin=<-5 -15 -5>, angularTarget=<1 0 0>}
      localJointOffset=<1 0 0>
     }]
    relativeSpaceBone {boneName=spine_05}
    simSpaceSettings {maxAngularAcceleration=40, maxAngularVelocity=8, simSpaceAngularAlpha=0.3}
   }
   n5 = Anim Dynamics - Bone: dyn_collar_bk_l_02(ComponentPose=@n4, Alpha=1, ExternalForce="0,0,0") {
    angularDampingOverride=0.99
    angularSpringConstant=200
    bAngularSpring=true
    bOverrideAngularDamping=true
    bUseGravityOverride=true
    simulationSpace=RootRelative
    boundBone {boneName=dyn_collar_bk_l_02}
    componentAppliedLinearAccClamp=<3000 3000 10000>
    componentLinearAccScale=<0.3 0.3 0.5>
    componentLinearVelScale=<0.3 0.3 1>
    gravityOverride {z=-980}
    physicsBodyDefinitions [{
      boundBone {boneName=dyn_collar_bk_l_02}
      boxExtents=<12 12 12>
      constraintSetup {angularLimitsMax=<0 15 0>, angularLimitsMin=<0 -15 0>, angularTarget=<1 0 0>}
      localJointOffset=<1 0 0>
     }]
    relativeSpaceBone {boneName=spine_05}
    simSpaceSettings {maxAngularAcceleration=40, maxAngularVelocity=8, simSpaceAngularAlpha=0.3}
   }
   n6 = Anim Dynamics - Bone: dyn_collar_bk_l_03(ComponentPose=@n5, Alpha=1, ExternalForce="0,0,0") {
    angularDampingOverride=0.99
    angularSpringConstant=200
    bAngularSpring=true
    bOverrideAngularDamping=true
    bUseGravityOverride=true
    simulationSpace=RootRelative
    boundBone {boneName=dyn_collar_bk_l_03}
    componentAppliedLinearAccClamp=<3000 3000 10000>
    componentLinearAccScale=<0.3 0.3 0.5>
    componentLinearVelScale=<0.3 0.3 1>
    gravityOverride {z=-980}
    physicsBodyDefinitions [{
      boundBone {boneName=dyn_collar_bk_l_03}
      boxExtents=<12 12 12>
      constraintSetup {angularLimitsMax=<0 15 5>, angularLimitsMin=<0 -15 -5>, angularTarget=<1 0 0>}
      localJointOffset=<1 0 0>
     }]
    relativeSpaceBone {boneName=spine_05}
    simSpaceSettings {maxAngularAcceleration=40, maxAngularVelocity=8, simSpaceAngularAlpha=0.3}
   }
   n7 = Anim Dynamics - Bone: dyn_collar_bk_02(ComponentPose=@n6, Alpha=1, ExternalForce="0,0,0") {
    angularDampingOverride=0.99
    angularSpringConstant=200
    bAngularSpring=true
    bOverrideAngularDamping=true
    bUseGravityOverride=true
    simulationSpace=RootRelative
    boundBone {boneName=dyn_collar_bk_02}
    componentAppliedLinearAccClamp=<3000 3000 10000>
    componentLinearAccScale=<0.3 0.3 0.5>
    componentLinearVelScale=<0.3 0.3 1>
    gravityOverride {z=-980}
    physicsBodyDefinitions [{
      boundBone {boneName=dyn_collar_bk_02}
      boxExtents=<12 12 12>
      constraintSetup {angularLimitsMax=<0 15 0>, angularLimitsMin=<0 -15 0>, angularTarget=<1 0 0>}
      localJointOffset=<1 0 0>
     }]
    relativeSpaceBone {boneName=spine_05}
    simSpaceSettings {maxAngularAcceleration=40, maxAngularVelocity=8, simSpaceAngularAlpha=0.3}
   }
   n8 = Anim Dynamics - Bone: dyn_collar_bk_03(ComponentPose=@n7, Alpha=1, ExternalForce="0,0,0") {
    angularDampingOverride=0.99
    angularSpringConstant=200
    bAngularSpring=true
    bOverrideAngularDamping=true
    bUseGravityOverride=true
    simulationSpace=RootRelative
    boundBone {boneName=dyn_collar_bk_03}
    componentAppliedLinearAccClamp=<3000 3000 10000>
    componentLinearAccScale=<0.3 0.3 0.5>
    componentLinearVelScale=<0.3 0.3 1>
    gravityOverride {z=-980}
    physicsBodyDefinitions [{
      boundBone {boneName=dyn_collar_bk_03}
      boxExtents=<12 12 12>
      constraintSetup {angularLimitsMax=<0 15 5>, angularLimitsMin=<0 -15 -5>, angularTarget=<1 0 0>}
      localJointOffset=<1 0 0>
     }]
    relativeSpaceBone {boneName=spine_05}
    simSpaceSettings {maxAngularAcceleration=40, maxAngularVelocity=8, simSpaceAngularAlpha=0.3}
   }
   n9 = Anim Dynamics - Bone: dyn_collar_bk_r_02(ComponentPose=@n8, Alpha=1, ExternalForce="0,0,0") {
    angularDampingOverride=0.99
    angularSpringConstant=200
    bAngularSpring=true
    bOverrideAngularDamping=true
    bUseGravityOverride=true
    simulationSpace=RootRelative
    boundBone {boneName=dyn_collar_bk_r_02}
    componentAppliedLinearAccClamp=<3000 3000 10000>
    componentLinearAccScale=<0.3 0.3 0.5>
    componentLinearVelScale=<0.3 0.3 1>
    gravityOverride {z=-980}
    physicsBodyDefinitions [{
      boundBone {boneName=dyn_collar_bk_r_02}
      boxExtents=<12 12 12>
      constraintSetup {angularLimitsMax=<0 15 0>, angularLimitsMin=<0 -15 0>, angularTarget=<1 0 0>}
      localJointOffset=<1 0 0>
     }]
    relativeSpaceBone {boneName=spine_05}
    simSpaceSettings {maxAngularAcceleration=40, maxAngularVelocity=8, simSpaceAngularAlpha=0.3}
   }
   n10 = Anim Dynamics - Bone: dyn_collar_bk_r_03(ComponentPose=@n9, Alpha=1, ExternalForce="0,0,0") {
    angularDampingOverride=0.99
    angularSpringConstant=200
    bAngularSpring=true
    bOverrideAngularDamping=true
    bUseGravityOverride=true
    simulationSpace=RootRelative
    boundBone {boneName=dyn_collar_bk_r_03}
    componentAppliedLinearAccClamp=<3000 3000 10000>
    componentLinearAccScale=<0.3 0.3 0.5>
    componentLinearVelScale=<0.3 0.3 1>
    gravityOverride {z=-980}
    physicsBodyDefinitions [{
      boundBone {boneName=dyn_collar_bk_r_03}
      boxExtents=<12 12 12>
      constraintSetup {angularLimitsMax=<0 15 5>, angularLimitsMin=<0 -15 -5>, angularTarget=<1 0 0>}
      localJointOffset=<1 0 0>
     }]
    relativeSpaceBone {boneName=spine_05}
    simSpaceSettings {maxAngularAcceleration=40, maxAngularVelocity=8, simSpaceAngularAlpha=0.3}
   }
   n11 = Anim Dynamics - Bone: dyn_collar_r_02(ComponentPose=@n10, Alpha=1, ExternalForce="0,0,0") {
    angularDampingOverride=0.99
    angularSpringConstant=200
    bAngularSpring=true
    bOverrideAngularDamping=true
    bUseGravityOverride=true
    simulationSpace=RootRelative
    boundBone {boneName=dyn_collar_r_02}
    componentAppliedLinearAccClamp=<3000 3000 10000>
    componentLinearAccScale=<0.3 0.3 0.5>
    componentLinearVelScale=<0.3 0.3 1>
    gravityOverride {z=-980}
    physicsBodyDefinitions [{
      boundBone {boneName=dyn_collar_r_02}
      boxExtents=<12 12 12>
      constraintSetup {angularLimitsMax=<0 15 0>, angularLimitsMin=<0 -15 0>, angularTarget=<1 0 0>}
      localJointOffset=<1 0 0>
     }]
    relativeSpaceBone {boneName=spine_05}
    simSpaceSettings {maxAngularAcceleration=40, maxAngularVelocity=8, simSpaceAngularAlpha=0.3}
   }
   n12 = Anim Dynamics - Bone: dyn_collar_r_03(ComponentPose=@n11, Alpha=1, ExternalForce="0,0,0") {
    angularDampingOverride=0.99
    angularSpringConstant=200
    bAngularSpring=true
    bOverrideAngularDamping=true
    bUseGravityOverride=true
    simulationSpace=RootRelative
    boundBone {boneName=dyn_collar_r_03}
    componentAppliedLinearAccClamp=<3000 3000 10000>
    componentLinearAccScale=<0.3 0.3 0.5>
    componentLinearVelScale=<0.3 0.3 1>
    gravityOverride {z=-980}
    physicsBodyDefinitions [{
      boundBone {boneName=dyn_collar_r_03}
      boxExtents=<12 12 12>
      constraintSetup {angularLimitsMax=<5 15 5>, angularLimitsMin=<-5 -15 -5>, angularTarget=<1 0 0>}
      localJointOffset=<1 0 0>
     }]
    relativeSpaceBone {boneName=spine_05}
    simSpaceSettings {maxAngularAcceleration=40, maxAngularVelocity=8, simSpaceAngularAlpha=0.3}
   }
   Output Pose(Result=Component To Local(ComponentPose=@n12)) {layerGroup=None}
  ]}
 EventGraph {nodes [
   Event BlueprintUpdateAnimation()
   n1 = GetOwnerSkeletalMeshComponent()
   n2 = BreakRotator(InRot=BreakTransform(InTransform=GetDeltaTransformFromRefPose(self=@n1, BoneName=head, BaseName=neck_02)).Rotation)
   n3 = Get Z percentageEmote() -> (Z percentageEmote:float)
   n4 = Get X percentageEmote() -> (X percentageEmote:float)
   n5 = Get Y percentageEmote() -> (Y percentageEmote:float)
   > Set headRotLimit(headRotLimit:Rotator=MakeRotator(
     Roll=float * float!(A=@n2, B=-1, C:double=@n4) -> (ReturnValue)
     Pitch=float * float(A=@n2.Pitch, B=-1, C=@n5)
     Yaw=float * float(A=@n2.Yaw, B=-1, C=@n3)
    )) -> (Output_Get:Rotator)
   n6 = GetPartSkeletalMeshComponent()
   > Set distanceToLeft(distanceToLeft:float=Get Distance Between Two Sockets!(
     Component:Skeletal Mesh Component Object Reference=@n6
     SocketOrBoneNameA:Name=lowerarm_l
     SocketSpaceA:ERelativeTransformSpace Enum=RTS_World
     SocketOrBoneNameB:Name=head
     SocketSpaceB:ERelativeTransformSpace Enum=RTS_World
     bRemapRange:bool=true
     InRangeMin:float=40
     InRangeMax:float=25
     OutRangeMin:float=1
     OutRangeMax:float=0
    ) -> (ReturnValue:float) {is=title K2Node_CallFunction}) -> (Output_Get:float)
   > Set DistanceToRight(DistanceToRight:float=Get Distance Between Two Sockets(
     Component=@n6
     SocketOrBoneNameA=lowerarm_r
     SocketSpaceA=RTS_World
     SocketOrBoneNameB=head
     SocketSpaceB=RTS_World
     bRemapRange=true
     InRangeMin=40
     InRangeMax=25
     OutRangeMin=1
     OutRangeMax=0
    )) -> (Output_Get:float)
   n7 = Event OnCharacterCustomizationLoaded()
   Is Valid(exec=[@n7.then, @n8.then], InputObject=@n7.CustomBaseMesh)
   > Set Body(Body:Skeletal Mesh Component Object Reference=@n7.CustomBaseMesh) -> (Output_Get:Skeletal Mesh Component Object Reference)
   n8 = Event BlueprintInitializeAnimation()
  ]}
]
variables [
 Body {type=SkeletalMeshComponent}
 distanceToLeft {type=number}
 DistanceToRight {type=number}
 headRotLimit {type=Rotator}
 isPistol {type=boolean}
 IsRifle {type=boolean}
 X percentageEmote {type=number, default=1}
 Y percentageEmote {type=number, default=1}
 Z percentageEmote {type=number, default=1}
]
