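# Physics asset for the TerryBroke medium male FaceAcc mesh, set up for a jointed, movable collar.
# Five collar chains (back, back-left, back-right, left, right) each run from a spine_05-locked base through three limited-swing segments, mostly boxes.
# Head, spine, clavicle and upper-arm shapes form a locked body frame around them.
# The left chain is asymmetric: its first joint is locked, lower joints have uneven swing limits, and its tip is a capsule.
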PhysicsAsset /BRCosmetics/Characters/Player/Male/Medium/Bodies/M_MED_TerryBroke/Meshes/Parts/M_MED_TerryBroke_FaceAcc_Physics
bodies {
 clavicle_l [{shapeType=Capsule, length=13.5488, radius=4.6763, shapeName=clavicle_l_capsule, center=<9.6043 0 0.4583>, rotation=<rot 0 -90 -92.7318>}]
 clavicle_r [{shapeType=Capsule, length=13.5488, radius=4.6763, shapeName=clavicle_r_capsule, center=<-9.6043 0 -0.4583>, rotation=<rot 0 -90 87.2682>}]
 dyn_collar_bk_01 [{
   shapeType=Box
   extentX=11.4328
   extentY=2.8253
   extentZ=16.3579
   shapeName=dyn_collar_bk_01_box
   center=<4.0835 -0.0416 -0.4352>
   rotation=<rot 0 2.527 -89.5932>
  }]
 dyn_collar_bk_02 [{
   shapeType=Box
   extentX=11.6278
   extentY=3.1059
   extentZ=19.9665
   shapeName=dyn_collar_bk_02_box
   center=<4.3164 -0.0033 -0.2955>
   rotation=<rot 0 178.9135 -89.8246>
  }]
 dyn_collar_bk_03 [{
   shapeType=Box
   extentX=11.5775
   extentY=4.1718
   extentZ=24.7005
   shapeName=dyn_collar_bk_03_box
   center=<3.9584 -0.005 0.3856>
   rotation=<rot 0 -179.9898 -89.9957>
  }]
 dyn_collar_bk_base [{shapeType=Box, extentX=1.01, extentY=1.01, extentZ=1.01, shapeName=dyn_collar_bk_base_box, rotation=<rot 0 -90 0>}]
 dyn_collar_bk_l_01 [{
   shapeType=Box
   extentX=10.6574
   extentY=2.4316
   extentZ=15.3133
   shapeName=dyn_collar_bk_l_01_box
   center=<5.1174 -0.5966 -0.3295>
   rotation=<rot 0 -164.4255 -81.5027>
  }]
 dyn_collar_bk_l_02 [{
   shapeType=Box
   extentX=9.3605
   extentY=3.173
   extentZ=19.0467
   shapeName=dyn_collar_bk_l_02_box
   center=<3.4209 -0.4374 -0.1669>
   rotation=<rot 0 -159.5799 -83.0504>
  }]
 dyn_collar_bk_l_03 [{
   shapeType=Box
   extentX=8.2895
   extentY=3.3282
   extentZ=21.1004
   shapeName=dyn_collar_bk_l_03_box
   center=<0.8518 0.3409 0.1637>
   rotation=<rot 0 -160.8957 -83.9455>
  }]
 dyn_collar_bk_l_base [{shapeType=Box, extentX=1.01, extentY=1.01, extentZ=1.01, shapeName=dyn_collar_bk_l_base_box, rotation=<rot 0 -90 0>}]
 dyn_collar_bk_r_01 [{
   shapeType=Box
   extentX=11.2441
   extentY=2.4508
   extentZ=14.9264
   shapeName=dyn_collar_bk_r_01_box
   center=<4.6958 0.4436 -0.2589>
   rotation=<rot 0 -14.1792 -81.2537>
  }]
 dyn_collar_bk_r_02 [{
   shapeType=Box
   extentX=8.8559
   extentY=3.1857
   extentZ=19.2498
   shapeName=dyn_collar_bk_r_02_box
   center=<3.2374 0.2374 -0.0518>
   rotation=<rot 0 -15.4684 -82.3721>
  }]
 dyn_collar_bk_r_03 [{
   shapeType=Box
   extentX=8.2712
   extentY=3.3784
   extentZ=21.1129
   shapeName=dyn_collar_bk_r_03_box
   center=<0.8772 -0.3509 0.1883>
   rotation=<rot 0 -18.4709 -83.657>
  }]
 dyn_collar_bk_r_base [{shapeType=Box, extentX=1.01, extentY=1.01, extentZ=1.01, shapeName=dyn_collar_bk_r_base_box, rotation=<rot 0 -90 0>}]
 dyn_collar_l_01 [{
   shapeType=Box
   extentX=9.998
   extentY=2.9581
   extentZ=14.5293
   shapeName=dyn_collar_l_01_box
   center=<3.8781 1.5869 -0.908>
   rotation=<rot 0 42.8553 -76.3189>
  }]
 dyn_collar_l_02 [{
   shapeType=Box
   extentX=14.136
   extentY=2.3539
   extentZ=20.7612
   shapeName=dyn_collar_l_02_box
   center=<0.4437 0.2526 -0.0319>
   rotation=<rot 0 1.4595 -72.5345>
  }]
 dyn_collar_l_03 [{
   shapeType=Capsule
   length=12.5943
   radius=3.6925
   shapeName=dyn_collar_l_03_capsule
   center=<3.4204 4.6946 -1.9079>
   rotation=<rot 0 0 -68.0968>
  }]
 dyn_collar_l_base [{shapeType=Box, extentX=1.01, extentY=1.01, extentZ=1.01, shapeName=dyn_collar_l_base_box, rotation=<rot 0 -90 0>}]
 dyn_collar_r_01 [{
   shapeType=Box
   extentX=10.2088
   extentY=2.495
   extentZ=14.3519
   shapeName=dyn_collar_r_01_box
   center=<3.9617 -1.6446 -0.6261>
   rotation=<rot 0 136.404 -77.1088>
  }]
 dyn_collar_r_02 [{
   shapeType=Box
   extentX=14.6027
   extentY=3.1942
   extentZ=20.5137
   shapeName=dyn_collar_r_02_box
   center=<1.1044 -0.1765 0.4174>
   rotation=<rot 0 177.2832 -77.1963>
  }]
 dyn_collar_r_03 [{
   shapeType=Box
   extentX=11.3995
   extentY=4.6285
   extentZ=25.4981
   shapeName=dyn_collar_r_03_box
   center=<2.5028 -0.0818 0.2192>
   rotation=<rot 0 173.9063 -70.8768>
  }]
 dyn_collar_r_base [{shapeType=Box, extentX=1.01, extentY=1.01, extentZ=1.01, shapeName=dyn_collar_r_base_box, rotation=<rot 0 -90 0>}]
 head [{shapeType=Sphere, radius=14.308, shapeName=head_sphere, center=<5.5343 2.6934 0>}]
 spine_05 [{shapeType=Capsule, length=24.6773, radius=6.3989, shapeName=spine_05_capsule}]
 upperarm_l [{
   shapeType=Capsule
   length=26.0014
   radius=7.4526
   shapeName=upperarm_l_capsule
   center=<15.5815 -3.3724 0.0726>
   rotation=<rot 0 -90 -92.7318>
  }]
 upperarm_r [{
   shapeType=Capsule
   length=28.7777
   radius=4.6763
   shapeName=upperarm_r_capsule
   center=<-15.6743 10.9749 1.8723>
   rotation=<rot 0 -90 87.2682>
  }]
}
constraints [
 {
  bone1Name=clavicle_r
  bone2Name=spine_05
  swing1LimitDegrees=45
  swing1Motion=Locked
  swing2LimitDegrees=45
  swing2Motion=Locked
  twistLimitDegrees=45
  twistMotion=Locked
 }
 {
  bone1Name=upperarm_r
  bone2Name=clavicle_r
  swing1LimitDegrees=45
  swing1Motion=Locked
  swing2LimitDegrees=45
  swing2Motion=Locked
  twistLimitDegrees=45
  twistMotion=Locked
 }
 {
  bone1Name=clavicle_l
  bone2Name=spine_05
  swing1LimitDegrees=45
  swing1Motion=Locked
  swing2LimitDegrees=45
  swing2Motion=Locked
  twistLimitDegrees=45
  twistMotion=Locked
 }
 {
  bone1Name=upperarm_l
  bone2Name=clavicle_l
  swing1LimitDegrees=45
  swing1Motion=Locked
  swing2LimitDegrees=45
  swing2Motion=Locked
  twistLimitDegrees=45
  twistMotion=Locked
 }
 {
  bone1Name=head
  bone2Name=spine_05
  swing1LimitDegrees=45
  swing1Motion=Locked
  swing2LimitDegrees=45
  swing2Motion=Locked
  twistLimitDegrees=45
  twistMotion=Locked
 }
 {
  bone1Name=dyn_collar_bk_base
  bone2Name=spine_05
  swing1LimitDegrees=45
  swing1Motion=Locked
  swing2LimitDegrees=45
  swing2Motion=Locked
  twistLimitDegrees=45
  twistMotion=Locked
 }
 {bone1Name=dyn_collar_bk_01, bone2Name=dyn_collar_bk_base, swing1LimitDegrees=11, swing2LimitDegrees=11, twistLimitDegrees=4}
 {bone1Name=dyn_collar_bk_02, bone2Name=dyn_collar_bk_01, swing1LimitDegrees=11, swing2LimitDegrees=11, twistLimitDegrees=4}
 {bone1Name=dyn_collar_bk_03, bone2Name=dyn_collar_bk_02, swing1LimitDegrees=11, swing2LimitDegrees=11, twistLimitDegrees=4}
 {
  bone1Name=dyn_collar_bk_l_base
  bone2Name=spine_05
  swing1LimitDegrees=45
  swing1Motion=Locked
  swing2LimitDegrees=45
  swing2Motion=Locked
  twistLimitDegrees=45
  twistMotion=Locked
 }
 {bone1Name=dyn_collar_bk_l_01, bone2Name=dyn_collar_bk_l_base, swing1LimitDegrees=11, swing2LimitDegrees=11, twistLimitDegrees=4}
 {bone1Name=dyn_collar_bk_l_02, bone2Name=dyn_collar_bk_l_01, swing1LimitDegrees=11, swing2LimitDegrees=11, twistLimitDegrees=4}
 {bone1Name=dyn_collar_bk_l_03, bone2Name=dyn_collar_bk_l_02, swing1LimitDegrees=11, swing2LimitDegrees=11, twistLimitDegrees=4}
 {
  bone1Name=dyn_collar_bk_r_base
  bone2Name=spine_05
  swing1LimitDegrees=45
  swing1Motion=Locked
  swing2LimitDegrees=45
  swing2Motion=Locked
  twistLimitDegrees=45
  twistMotion=Locked
 }
 {bone1Name=dyn_collar_bk_r_01, bone2Name=dyn_collar_bk_r_base, swing1LimitDegrees=11, swing2LimitDegrees=11, twistLimitDegrees=4}
 {bone1Name=dyn_collar_bk_r_02, bone2Name=dyn_collar_bk_r_01, swing1LimitDegrees=11, swing2LimitDegrees=11, twistLimitDegrees=4}
 {bone1Name=dyn_collar_bk_r_03, bone2Name=dyn_collar_bk_r_02, swing1LimitDegrees=11, swing2LimitDegrees=11, twistLimitDegrees=4}
 {
  bone1Name=dyn_collar_l_base
  bone2Name=spine_05
  swing1LimitDegrees=45
  swing1Motion=Locked
  swing2LimitDegrees=45
  swing2Motion=Locked
  twistLimitDegrees=45
  twistMotion=Locked
 }
 {
  bone1Name=dyn_collar_l_01
  bone2Name=dyn_collar_l_base
  swing1LimitDegrees=11
  swing1Motion=Locked
  swing2LimitDegrees=11
  swing2Motion=Locked
  twistLimitDegrees=4
  twistMotion=Locked
 }
 {bone1Name=dyn_collar_l_02, bone2Name=dyn_collar_l_01, swing1LimitDegrees=8, swing2LimitDegrees=22, twistLimitDegrees=4}
 {bone1Name=dyn_collar_l_03, bone2Name=dyn_collar_l_02, swing1LimitDegrees=3.68, swing2LimitDegrees=30.24, twistLimitDegrees=4}
 {
  bone1Name=dyn_collar_r_base
  bone2Name=spine_05
  swing1LimitDegrees=45
  swing1Motion=Locked
  swing2LimitDegrees=45
  swing2Motion=Locked
  twistLimitDegrees=45
  twistMotion=Locked
 }
 {bone1Name=dyn_collar_r_01, bone2Name=dyn_collar_r_base, swing1LimitDegrees=11, swing2LimitDegrees=11, twistLimitDegrees=4}
 {bone1Name=dyn_collar_r_02, bone2Name=dyn_collar_r_01, swing1LimitDegrees=11, swing2LimitDegrees=11, twistLimitDegrees=4}
 {bone1Name=dyn_collar_r_03, bone2Name=dyn_collar_r_02, swing1LimitDegrees=11, swing2LimitDegrees=11, twistLimitDegrees=4}
]
